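# Demo material for the head of the medium male player robot body M_MED_Robot_01, using Default Lit shading.
# It is enabled for skeletal meshes and mesh deformers.
# The graph holds the head's diffuse and normal texture samples plus three scalar constants, with no parameters.
# The links from these nodes to the material outputs are not recorded here.
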Material /Game/Characters/Player/Male/Medium/Bodies/M_MED_Robot_01/Demo/Materials/Demo_M_MED_Robot_Head
graph {
 properties {bUsedWithMeshDeformer=true, bUsedWithSkeletalMesh=true, pixelDepthOffsetMode=PDOM_Legacy, usedShadingModels=Default Lit}
 nodes [
  Constant() {r=0.5}
  Constant() {r=1}
  Constant() {r=0.4}
  TextureSample() {samplerType=SAMPLERTYPE_Normal, texture=/Game/Characters/Player/Male/Medium/Bodies/M_MED_Robot_01/Demo/Textures/T_Demo_Robot_Head_N}
  TextureSample() {texture=/Game/Characters/Player/Male/Medium/Bodies/M_MED_Robot_01/Demo/Textures/T_Demo_Robot_Head_D}
 ]
}
